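# Sound cue for the swing of the IndieBucket cosmetic pickaxe, mixing close and medium attenuation branches.
# The close branch makes a weighted random pick between a plain swing and a less likely variant that mixes a "Lyr" layer with an independently randomized swing.
# The medium branch randomizes among shared distant fire axe swings.
# Each branch passes through a modulator, and a melee concurrency set governs the cue.
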SoundCue /BRCosmetics/Sounds/PickAxes/IndieBucket/PickaxeSwing_IndieBucket_Cue
root=Mixer_0
nodes [
 {id=Attenuation_0, children [Modulator_0], properties {attenuationSettings=/Game/Sounds/Attenuation/Fort_Weapon_Pickaxe_Close_Att}}
 {id=Attenuation_1, children [Modulator_1], properties {attenuationSettings=/Game/Sounds/Attenuation/Fort_Weapon_Pickaxe_Medium_Att}}
 {id=Mixer_0, children [Attenuation_0, Attenuation_1], properties {inputVolume [1,1]}}
 {id=Mixer_1, children [Random_2,Random_3], properties {inputVolume [0.75,1]}}
 {id=Modulator_0, children [Random_4]}
 {id=Modulator_1, children [Random_1]}
 {id=Random_0, children [WavePlayer_0,WavePlayer_1,WavePlayer_2,WavePlayer_3,WavePlayer_4,WavePlayer_5], properties {weights [1,1,1,1,1,1]}}
 {
  id=Random_1
  children [WavePlayer_6, WavePlayer_7, WavePlayer_8, WavePlayer_9, WavePlayer_10, WavePlayer_11]
  properties {weights [1,1,1,1,1,1]}
 }
 {
  id=Random_2
  children [WavePlayer_12, WavePlayer_13, WavePlayer_14, WavePlayer_15, WavePlayer_16, WavePlayer_17]
  properties {weights [1,1,1,1,1,1]}
 }
 {
  id=Random_3
  children [WavePlayer_18, WavePlayer_19, WavePlayer_20, WavePlayer_21, WavePlayer_22, WavePlayer_23]
  properties {weights [1,1,1,1,1,1]}
 }
 {id=Random_4, children [Random_0,Mixer_1], properties {bRandomizeWithoutReplacement=false, weights [1,0.6]}}
 {id=WavePlayer_0, properties {soundWaveAssetPtr=&/Pickaxe_IndieBucket_Swing_01}}
 {id=WavePlayer_1, properties {soundWaveAssetPtr=&/Pickaxe_IndieBucket_Swing_06}}
 {id=WavePlayer_10, properties {soundWaveAssetPtr=/Game/Athena/Sounds/Weapons/Fire_Axe_Swing_Distant_05}}
 {id=WavePlayer_11, properties {soundWaveAssetPtr=/Game/Athena/Sounds/Weapons/Fire_Axe_Swing_Distant_06}}
 {id=WavePlayer_12, properties {soundWaveAssetPtr=&/Pickaxe_IndieBucket_Swing_01_Lyr}}
 {id=WavePlayer_13, properties {soundWaveAssetPtr=&/Pickaxe_IndieBucket_Swing_06_Lyr}}
 {id=WavePlayer_14, properties {soundWaveAssetPtr=&/Pickaxe_IndieBucket_Swing_05_Lyr}}
 {id=WavePlayer_15, properties {soundWaveAssetPtr=&/Pickaxe_IndieBucket_Swing_04_Lyr}}
 {id=WavePlayer_16, properties {soundWaveAssetPtr=&/Pickaxe_IndieBucket_Swing_03_Lyr}}
 {id=WavePlayer_17, properties {soundWaveAssetPtr=&/Pickaxe_IndieBucket_Swing_02_Lyr}}
 {id=WavePlayer_18, properties {soundWaveAssetPtr=&Pickaxe_IndieBucket_Swing_06}}
 {id=WavePlayer_19, properties {soundWaveAssetPtr=&/Pickaxe_IndieBucket_Swing_05}}
 {id=WavePlayer_2, properties {soundWaveAssetPtr=&/Pickaxe_IndieBucket_Swing_02}}
 {id=WavePlayer_20, properties {soundWaveAssetPtr=&/Pickaxe_IndieBucket_Swing_04}}
 {id=WavePlayer_21, properties {soundWaveAssetPtr=&/Pickaxe_IndieBucket_Swing_03}}
 {id=WavePlayer_22, properties {soundWaveAssetPtr=&Pickaxe_IndieBucket_Swing_02}}
 {id=WavePlayer_23, properties {soundWaveAssetPtr=&Pickaxe_IndieBucket_Swing_01}}
 {id=WavePlayer_3, properties {soundWaveAssetPtr=&Pickaxe_IndieBucket_Swing_03}}
 {id=WavePlayer_4, properties {soundWaveAssetPtr=&Pickaxe_IndieBucket_Swing_04}}
 {id=WavePlayer_5, properties {soundWaveAssetPtr=&Pickaxe_IndieBucket_Swing_05}}
 {id=WavePlayer_6, properties {soundWaveAssetPtr=/Game/Athena/Sounds/Weapons/Fire_Axe_Swing_Distant_01}}
 {id=WavePlayer_7, properties {soundWaveAssetPtr=/Game/Athena/Sounds/Weapons/Fire_Axe_Swing_Distant_02}}
 {id=WavePlayer_8, properties {soundWaveAssetPtr=/Game/Athena/Sounds/Weapons/Fire_Axe_Swing_Distant_03}}
 {id=WavePlayer_9, properties {soundWaveAssetPtr=/Game/Athena/Sounds/Weapons/Fire_Axe_Swing_Distant_04}}
]
properties {
 audioPropertiesSheet=/BRCosmetics/Sounds/Settings/PropertySheets/Pickaxe/APS_Cosmetic_Pickaxe_Swing
 maxDistance=3000
 soundClassObject=&Fort_WeaponMelee_PSM
 volumeMultiplier=0.3
 concurrencySet [&Player_Melee_Weapon_Conc]
}
